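; Cosmetic action definition for the PawJasmine sword reactivity cosmetic modifier in Battle Royale.
; It overrides GetActionInstanceClass to return the CA_PawJasmineSwordReactivity Blueprint class as the action instance class.
; It holds no reactivity logic itself and only names that class.
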(fn GetActionInstanceClass
  (return "/BRCosmetics/Athena/Cosmetics/CosmeticModifiers/CosmeticActions/CA_PawJasmineSwordReactivity.CA_PawJasmineSwordReactivity_C"))
