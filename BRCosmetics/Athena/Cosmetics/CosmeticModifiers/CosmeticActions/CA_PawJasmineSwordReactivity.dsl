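; Cosmetic action that drives material reactivity on the player's backpack mesh, named for the PawJasmine sword.
; On initialize, it caches the pawn's Backpack skeletal mesh and pulls opacity and time step float curves from its CAD_PawJasmineSwordReactivity definition.
; Start and end toggle the reactivity material parameter.
; Each update samples both curves by action progress to set scrolling and swoosh parameters, which end resets using zero progress.
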(event EventOnInitialize
  (params (CosmeticActionDefinition :type Cosmetic%20Action%20Definition%20Object%20Reference) (InActionContext :type Cosmetic%20Action%20Context%20Structure%20%28by%20ref%29))
  (Utilities|FlowControl|Sequence
    (:then_0
      (bind _asfort_player_pawn (Utilities|Casting|CastToFortPlayerPawn (Utilities|Struct|BreakCosmeticActionContext (Variables|CosmeticAction|GetContext))))
      (bind _output_get (Variables|Default|SetTargetPlayerPawn _asfort_player_pawn))
      (Variables|Default|SetTargetSkeletalMesh (Customization|GetSkeletalMeshforPartType _output_get "Backpack" :!class /Script/FortniteGame.FortPlayerPawn)))
    (:then_1
      (bind _ascad_pawjasmineswordreactivity (Utilities|Casting|CastToCAD_PawJasmineSwordReactivity CosmeticActionDefinition))
      (bind _output_get_2 (Variables|Data|SetActionDefinition _ascad_pawjasmineswordreactivity))
      (Utilities|FlowControl|Sequence
        (:then_0
          (bind _opacityfloatcurve (Variables|/BRCosmetics/Athena/Cosmetics/CosmeticModifiers/CosmeticActions/CAD_PawJasmineSwordReactivity.CAD_PawJasmineSwordReactivity_C|GetOpacityFloatCurve _output_get_2)
            (:then
              (Variables|RichCurveTracks|SetOpactity_Curve _opacityfloatcurve))
            (:else)))
        (:then_1
          (bind _timestepfloatcurve (Variables|/BRCosmetics/Athena/Cosmetics/CosmeticModifiers/CosmeticActions/CAD_PawJasmineSwordReactivity.CAD_PawJasmineSwordReactivity_C|GetTimestepFloatCurve _output_get_2)
            (:then
              (Variables|RichCurveTracks|SetTimeStep_Curve _timestepfloatcurve))
            (:else)))))))

(event EventOnStartAction
  (params (InReactionEventContext :type Cosmetic%20Reaction%20Event%20Context%20Structure%20%28by%20ref%29))
  (bind _targetskeletalmesh (Variables|Default|GetTargetSkeletalMesh)
    (:then
      (Rendering|Material|SetScalarParameterValueonMaterials _targetskeletalmesh "Enable Reactivity?" 1.0))
    (:else)))

(event EventOnEndAction
  (params (InReactionEventContext :type Cosmetic%20Reaction%20Event%20Context%20Structure%20%28by%20ref%29))
  (bind _targetskeletalmesh (Variables|Default|GetTargetSkeletalMesh)
    (:then
      (Rendering|Material|SetScalarParameterValueonMaterials _targetskeletalmesh "Enable Reactivity?" 0.0)
      ; Restore the updated material parameters back to the initial state
      (CallFunction|UpdateMaterialParameters 0.0))
    (:else)))

(event EventOnUpdateAction
  (params (DeltaTime :type Float%20%28single-precision%29))
  (Variables|Default|GetTargetSkeletalMesh
    (:then
      (CallFunction|UpdateMaterialParameters (CallFunction|GetActionProgress)))
    (:else)))

(fn ShouldUpdateAction
  (return true))

(fn UpdateMaterialParameters
  (params (ActionProgress :type Float%20%28double-precision%29))
  (bind _actionprogress (Variables|GetActionProgress))
  (bind _opactity_curve (Variables|RichCurveTracks|GetOpactity_Curve))
  (bind _returnvalue (Cosmetics|GetCurveFloatValuebyLerpValue _opactity_curve _actionprogress))
  (bind _targetskeletalmesh (Variables|Default|GetTargetSkeletalMesh))
  (Utilities|FlowControl|Sequence
    (:then_0
      (Rendering|Material|SetScalarParameterValueonMaterials (Variables|Default|GetTargetSkeletalMesh) "Scrolling - Time Step" (Cosmetics|GetCurveFloatValuebyLerpValue (Variables|RichCurveTracks|GetTimeStep_Curve) (Variables|GetActionProgress))))
    (:then_1
      (Rendering|Material|SetScalarParameterValueonMaterials _targetskeletalmesh "Scrolling - Intensity" _returnvalue)
      (Rendering|Material|SetScalarParameterValueonMaterials _targetskeletalmesh "Swoosh Intensity" _returnvalue))))
